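# Looping hover sound cue for the Athena supply drop drone, classed as Athena ambient audio with its own hover attenuation asset.
# A single looping wave feeds an Enveloper that fades volume in and shapes pitch over the start.
# A smoothed, intensified Doppler node follows, then a Modulator adds slight random pitch variation.
# All four children of the root quality-level node share this one chain, so every quality level sounds the same.
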SoundCue /Game/Athena/Sounds/SupplyDrops/Drone/SC_SupplyDrop_Drone_Hover_Loop_Cue
root=QualityLevel_0
nodes [
 {id=Doppler_0, children [Enveloper_2], properties {bUseSmoothing=true, dopplerIntensity=3}}
 {
  id=Enveloper_2
  children [WavePlayer_5]
  properties {pitchCurve {editorCurveData {keys [<at 0.0062 0.4928>, {time=1}]}}, volumeCurve {editorCurveData {keys [<at 0 0 RCIM_Cubic>,<at 1 1>]}}}
 }
 {id=Modulator_0, children [Doppler_0], properties {pitchMax=1.1, pitchMin=0.9, volumeMax=1, volumeMin=1}}
 {id=QualityLevel_0, children [Modulator_0,Modulator_0,Modulator_0,Modulator_0]}
 {id=WavePlayer_5, properties {bLooping=true, soundWaveAssetPtr=&/Athena_SupplyDrop_Drone_Loop_04}}
]
properties {
 attenuationSettings=&/SupplyDrop_Drone_Hover_Loop_Att
 maxDistance=4000
 soundClassObject=&Fort_Ambient_Athena
 volumeMultiplier=1.504
 attenuationOverrides {bEnableListenerFocus=true, nonFocusVolumeAttenuation=0.5}
 concurrencyOverrides {maxCount=4, resolutionRule=StopQuietest}
}
