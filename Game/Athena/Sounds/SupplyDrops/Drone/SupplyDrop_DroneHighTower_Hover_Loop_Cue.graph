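# Sound cue for the looping hover sound of the HighTower supply drop drone in Athena.
# A looping wave player feeds a smoothed, intensified Doppler node, which feeds an attenuation node using the drone hover attenuation asset.
# The cue uses the Athena ambient sound class and keeps playing when silent.
# It also carries override values for listener-focus attenuation and for concurrency that stops the quietest instance.
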SoundCue /Game/Athena/Sounds/SupplyDrops/Drone/SupplyDrop_DroneHighTower_Hover_Loop_Cue
root=Attenuation_0
nodes [
 {
  id=Attenuation_0
  children [Doppler_0]
  properties {attenuationSettings=&/SupplyDrop_Drone_Hover_Loop_Att}
 }
 {id=Doppler_0, children [WavePlayer_5], properties {bUseSmoothing=true, dopplerIntensity=2}}
 {id=WavePlayer_5, properties {bLooping=true, soundWaveAssetPtr=&/Athena_SupplyDrop_Drone_Loop_04}}
]
properties {
 maxDistance=4000
 soundClassObject=&Fort_Ambient_Athena
 virtualizationMode=PlayWhenSilent
 volumeMultiplier=1.039
 attenuationOverrides {bEnableListenerFocus=true, nonFocusVolumeAttenuation=0.5}
 concurrencyOverrides {maxCount=4, resolutionRule=StopQuietest}
}
