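# Sound cue for the supply drop drone's despawn in Athena, classed as an impact sound with gadget-working attenuation.
# A quality-level node sends its first slot to the standard waves and its other three slots to a shared branch of Opt variants.
# Each branch randomly picks one of two equally weighted despawn waves and passes it through a modulator.
# A concurrency override blocks new instances once the limit is reached.
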SoundCue /Game/Athena/Sounds/SupplyDrops/Drone/SupplyDrop_Drone_Despawn_Cue
root=QualityLevel_0
nodes [
 {id=Modulator_0, children [Random_1]}
 {id=Modulator_1, children [Random_0]}
 {id=QualityLevel_0, children [Modulator_1,Modulator_0,Modulator_0,Modulator_0]}
 {id=Random_0, children [WavePlayer_1,WavePlayer_2], properties {weights [1,1]}}
 {id=Random_1, children [WavePlayer_0,WavePlayer_3], properties {weights [1,1]}}
 {id=WavePlayer_0, properties {soundWaveAssetPtr=/Game/Athena/Sounds/SupplyDrops/Drone/OPT/Athena_SupplyDrop_Drone_Despawn_02_Opt}}
 {id=WavePlayer_1, properties {soundWaveAssetPtr=&/Athena_SupplyDrop_Drone_Despawn_01}}
 {id=WavePlayer_2, properties {soundWaveAssetPtr=&/Athena_SupplyDrop_Drone_Despawn_02}}
 {id=WavePlayer_3, properties {soundWaveAssetPtr=/Game/Athena/Sounds/SupplyDrops/Drone/OPT/Athena_SupplyDrop_Drone_Despawn_01_Opt}}
]
properties {
 attenuationSettings=/Game/Sounds/Attenuation/Fort_GadgetWorking_Att
 bOverrideConcurrency=true
 maxDistance=3100
 soundClassObject=&Fort_Impacts
 volumeMultiplier=1
 concurrencyOverrides {resolutionRule=PreventNew}
}
