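# Sound cue for a supply drop drone being destroyed in Athena.
# A Random node picks one of three equally weighted destruction waves and feeds a Modulator that varies pitch and volume at its default ranges.
# It plays in the large explosives sound class with a dedicated attenuation asset, and overrides concurrency to cap simultaneous instances.
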SoundCue /Game/Athena/Sounds/SupplyDrops/Drone/SupplyDrop_Drone_Destroyed_Cue
root=Modulator_1
nodes [
 {id=Modulator_1, children [Random_1]}
 {id=Random_1, children [WavePlayer_0,WavePlayer_4,WavePlayer_5], properties {weights [1,1,1]}}
 {id=WavePlayer_0, properties {soundWaveAssetPtr=&/Athena_SupplyDrop_Drone_Destroyed_01}}
 {id=WavePlayer_4, properties {soundWaveAssetPtr=&/Athena_SupplyDrop_Drone_Destroyed_02}}
 {id=WavePlayer_5, properties {soundWaveAssetPtr=&/Athena_SupplyDrop_Drone_Destroyed_03}}
]
properties {
 attenuationSettings=&/SupplyDrop_Drone_Destroyed_Att
 bOverrideConcurrency=true
 maxDistance=12500
 soundClassObject=/Game/Sounds/FortSoundClassesAndModes/Fort_Explosives_Large_PSM
 volumeMultiplier=1
 attenuationOverrides {falloffDistance=8000, attenuationShapeExtents {x=0}}
 concurrencyOverrides {maxCount=4}
}
