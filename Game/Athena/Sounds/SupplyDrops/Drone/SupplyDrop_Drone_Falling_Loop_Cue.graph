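# Falling-loop sound cue for the Athena supply drop drone, mixing two separately attenuated branches.
# The body branch layers looping wind and meteor-fall waves through an enveloper that ramps volume up at the start.
# The beep branch randomly picks one of six non-looping drone beeps, doubles its pitch and uses the drone-destroyed attenuation.
# It plays in the Athena ambient sound class, with concurrency set to prevent new instances.
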SoundCue /Game/Athena/Sounds/SupplyDrops/Drone/SupplyDrop_Drone_Falling_Loop_Cue
root=Mixer_1
nodes [
 {id=Attenuation_0, children [Enveloper_0], properties {attenuationSettings=&/SupplyDrop_Drone_Athena_Att}}
 {
  id=Attenuation_1
  children [Modulator_3]
  properties {attenuationSettings=&/SupplyDrop_Drone_Destroyed_Att}
 }
 {
  id=Enveloper_0
  children [Mixer_0]
  properties {pitchCurve {editorCurveData {keys [{time=0}, {time=1}]}}, volumeCurve {editorCurveData {keys [<at 0 0.65>,<at 0.25 1>]}}}
 }
 {id=Mixer_0, children [WavePlayer_2,WavePlayer_0], properties {inputVolume [1,0.65]}}
 {id=Mixer_1, children [Attenuation_0, Attenuation_1], properties {inputVolume [3,5]}}
 {id=Modulator_3, children [Random_2], properties {pitchMax=2, pitchMin=2, volumeMax=1, volumeMin=1}}
 {
  id=Random_2
  children [WavePlayer_7, WavePlayer_8, WavePlayer_9, WavePlayer_10, WavePlayer_11, WavePlayer_12]
  properties {bShouldExcludeFromBranchCulling=true, weights [1,1,1,1,1,1]}
 }
 {id=WavePlayer_0, properties {bLooping=true, soundWaveAssetPtr=/Game/Sounds/Fort_GamePlay_Sounds/Meteor/Meteor_Carmine_FallingLoop_03}}
 {id=WavePlayer_10, properties {soundWaveAssetPtr=&/Athena_SupplyDrop_Drone_Beep_04}}
 {id=WavePlayer_11, properties {soundWaveAssetPtr=&/Athena_SupplyDrop_Drone_Beep_05}}
 {id=WavePlayer_12, properties {soundWaveAssetPtr=&/Athena_SupplyDrop_Drone_Beep_06}}
 {id=WavePlayer_2, properties {bLooping=true, soundWaveAssetPtr=/Game/Sounds/Fort_UI_Sounds/Replay/Drone_Altitude_WindLoop}}
 {id=WavePlayer_7, properties {soundWaveAssetPtr=&/Athena_SupplyDrop_Drone_Beep_01}}
 {id=WavePlayer_8, properties {soundWaveAssetPtr=&/Athena_SupplyDrop_Drone_Beep_02}}
 {id=WavePlayer_9, properties {soundWaveAssetPtr=&/Athena_SupplyDrop_Drone_Beep_03}}
]
properties {
 maxDistance=12500
 soundClassObject=&Fort_Ambient_Athena
 volumeMultiplier=0.501
 attenuationOverrides {bEnableListenerFocus=true, nonFocusVolumeAttenuation=0.5}
 concurrencyOverrides {resolutionRule=PreventNew}
}
