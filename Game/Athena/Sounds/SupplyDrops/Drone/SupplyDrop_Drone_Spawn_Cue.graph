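# Sound cue for the spawn sound of a supply drop drone in Athena, classed as Athena ambient audio and spatialized with gadget-working attenuation.
# A quality-level root selects among three full-size random variants, three optimized random variants, or one optimized wave shared by the last two levels.
# Each branch passes through a modulator for variation.
# A concurrency override allows one instance per owner and blocks new plays while it runs.
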SoundCue /Game/Athena/Sounds/SupplyDrops/Drone/SupplyDrop_Drone_Spawn_Cue
root=QualityLevel_0
nodes [
 {id=Modulator_0, children [Random_0]}
 {id=Modulator_1, children [Random_1]}
 {id=Modulator_2, children [WavePlayer_6]}
 {id=QualityLevel_0, children [Modulator_0,Modulator_1,Modulator_2,Modulator_2]}
 {id=Random_0, children [WavePlayer_0,WavePlayer_1,WavePlayer_2], properties {weights [1,1,1]}}
 {id=Random_1, children [WavePlayer_3,WavePlayer_4,WavePlayer_5], properties {weights [1,1,1]}}
 {id=WavePlayer_0, properties {soundWaveAssetPtr=&/Athena_SupplyDrop_Drone_Spawn_03}}
 {id=WavePlayer_1, properties {soundWaveAssetPtr=&/Athena_SupplyDrop_Drone_Spawn_01}}
 {id=WavePlayer_2, properties {soundWaveAssetPtr=&/Athena_SupplyDrop_Drone_Spawn_02}}
 {id=WavePlayer_3, properties {soundWaveAssetPtr=/Game/Athena/Sounds/SupplyDrops/Drone/OPT/Athena_SupplyDrop_Drone_Spawn_01_Opt}}
 {id=WavePlayer_4, properties {soundWaveAssetPtr=/Game/Athena/Sounds/SupplyDrops/Drone/OPT/Athena_SupplyDrop_Drone_Spawn_02_Opt}}
 {id=WavePlayer_5, properties {soundWaveAssetPtr=/Game/Athena/Sounds/SupplyDrops/Drone/OPT/Athena_SupplyDrop_Drone_Spawn_03_Opt}}
 {id=WavePlayer_6, properties {soundWaveAssetPtr=&Athena_SupplyDrop_Drone_Spawn_03_Opt}}
]
properties {
 attenuationSettings=/Game/Sounds/Attenuation/Fort_GadgetWorking_Att
 bOverrideConcurrency=true
 maxDistance=3100
 soundClassObject=&Fort_Ambient_Athena
 volumeMultiplier=1.143
 concurrencyOverrides {bLimitToOwner=true, maxCount=1, resolutionRule=PreventNew}
}
